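# Physics asset for the bandage on the medium male BattleHound Fierce player cosmetic body, defining its simulation shapes.
# Small capsules form left and right two-link chains branching from a central bandage body.
# Constraints give each joint moderate swing and tight twist limits, so the strips bend but barely twist.
# A much larger capsule on mantle_2 has no constraints, apparently serving as a collision volume for the strips.
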PhysicsAsset /BRCosmetics/Characters/Player/Male/Medium/Bodies/M_MED_BattleHound_Fierce/Meshes/M_MED_BattleHound_Fierce_Bandage_Physics
bodies {
 bandage [{shapeType=Capsule, length=1, radius=1, shapeName=bandage_capsule}]
 bandage_l_1 [{shapeType=Capsule, length=2, radius=1.5, shapeName=bandage_l_1_capsule, center=<1.8755 0 0>, rotation=<rot 90 0 0>}]
 bandage_l_2 [{shapeType=Capsule, length=2.75, radius=1.75, shapeName=bandage_l_2_capsule, center=<2.9744 -0.3254 0>, rotation=<rot 90 0 0>}]
 bandage_r_1 [{shapeType=Capsule, length=2.375, radius=1.625, shapeName=bandage_r_1_capsule, center=<3.565 0 0>, rotation=<rot 90 0 0>}]
 bandage_r_2 [{shapeType=Capsule, length=2.625, radius=1.875, shapeName=bandage_r_2_capsule, center=<3.6162 0 0.3319>, rotation=<rot 90 0 0>}]
 mantle_2 [{
   shapeType=Capsule
   length=6.6483
   radius=19.9811
   shapeName=mantle_2_capsule
   center=<2.8916 0 0.0827>
   rotation=<rot 90 -179.9605 -179.9603>
  }]
}
constraints [
 {bone1Name=bandage, bone2Name=bandage_l_1, swing1LimitDegrees=40, swing2LimitDegrees=40, twistLimitDegrees=5}
 {bone1Name=bandage, bone2Name=bandage_r_1, swing1LimitDegrees=40, swing2LimitDegrees=40, twistLimitDegrees=5}
 {bone1Name=bandage_l_1, bone2Name=bandage_l_2, swing1LimitDegrees=40, swing2LimitDegrees=40, twistLimitDegrees=5}
 {bone1Name=bandage_r_1, bone2Name=bandage_r_2, swing1LimitDegrees=40, swing2LimitDegrees=40, twistLimitDegrees=5}
]
