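# Control Rig for the Battle Hound Fierce medium male body's cape, which adjusts the mantle bone chain to follow the spine.
# Parent constraints tie mantle_2, mantle_3 and mantle_4 to spine_04, spine_03 and spine_02, then fixed offsets shift mantle_1 and rotate mantle_3 with its children.
# A spherical pose reader on spine_05, relative to spine_01, weights a final translation of mantle_2 alone; its debug drawing is left on.
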ControlRigBlueprint /BRCosmetics/Characters/Player/Male/Medium/Bodies/M_MED_BattleHound_Fierce/Meshes/M_MED_BattleHound_Fierce_Cape_CtrlRig
graphs [
 RigVMFunctionLibrary {links [], nodes []}
 RigVMModel {nodes [
   RigUnit_BeginExecution = &RigUnit_BeginExecution()
   &RigUnit_ParentConstraint(
    AdvancedSettings=(InterpolationType=Average,RotationOrderForFilter=XZY)
    Child=(Type=Bone,Name="mantle_2")
    Parents=((Item=(Type=Bone,Name="spine_04"),Weight=1))
    Weight=1
    ExecutePin=@RigUnit_BeginExecution.ExecutePin
   )
   > &RigUnit_ParentConstraint(
    AdvancedSettings=(InterpolationType=Average,RotationOrderForFilter=XZY)
    Child=(Type=Bone,Name="mantle_3")
    Parents=((Item=(Type=Bone,Name="spine_03"),Weight=1))
    Weight=1
   )
   > &RigUnit_ParentConstraint(
    AdvancedSettings=(InterpolationType=Average,RotationOrderForFilter=XZY)
    Child=(Type=Bone,Name="mantle_4")
    Parents=((Item=(Type=Bone,Name="spine_02"),Weight=1))
    Weight=1
   )
   > &RigUnit_OffsetTransformForItem(
    Item=(Type=Bone,Name="mantle_1")
    OffsetTransform=(Rotation=(X=0,Y=0,Z=0,W=1),Translation=(X=0,Y=2.5,Z=0),Scale3D=(X=1,Y=1,Z=1))
    Weight=1
    bPropagateToChildren=false
   )
   > &RigUnit_OffsetTransformForItem(
    Item=(Type=Bone,Name="mantle_3")
    OffsetTransform=(Rotation=(X=0,Y=0,Z=-0.0436,W=0.999),Translation=(X=0,Y=0,Z=0),Scale3D=(X=1,Y=1,Z=1))
    Weight=1
    bPropagateToChildren=True
   )
   > n1 = &RigUnit_SphericalPoseReader(
    ActiveRegionScaleFactors=(PositiveWidth=1,NegativeWidth=1,PositiveHeight=1,NegativeHeight=1)
    ActiveRegionSize=0.27
    Debug=(bDrawDebug=True,bDraw2D=False,bDrawLocalAxes=False,DebugScale=25,DebugSegments=20,DebugThickness=0.25)
    DriverAxis=(X=0,Y=-1,Z=0)
    DriverItem=(Type=Bone,Name="spine_05")
    FalloffRegionScaleFactors=(PositiveWidth=1,NegativeWidth=1,PositiveHeight=1,NegativeHeight=1)
    FalloffSize=0.5
    FlipHeightScaling=False
    FlipWidthScaling=False
    OptionalParentItem=(Type=Bone,Name="spine_01")
    RotationOffset=(X=180,Y=0,Z=0)
   )
   > &RigUnit_OffsetTransformForItem(
    Item=(Type=Bone,Name="mantle_2")
    OffsetTransform=(Rotation=(X=0,Y=0,Z=0,W=1),Translation=(X=0,Y=4,Z=0),Scale3D=(X=1,Y=1,Z=1))
    bPropagateToChildren=false
    Weight=@n1.OutputParam
   )
  ]}
]
hierarchy [
 root {loc=<0 0 0>, rot=<0 0 0 1>}
 pelvis {parent=root, loc=<0 -0.7071 93.1191>, rot=<0.0127 0.707 -0.0127 -0.707>}
 spine_01 {parent=pelvis, loc=<6.7494 -0.6388 0>, rot=<0 0 0.034 -0.9994>, scale=<1.0001 1 1>}
 spine_02 {parent=spine_01, loc=<7.2518 -0.035 0>, rot=<0 0 0.0238 -0.9997>, scale=<0.9999 1 1>}
 spine_03 {parent=spine_02, loc=<6.0104 0 0>, rot=<0 0 0.0149 -0.9999>}
 spine_04 {parent=spine_03, loc=<8.1024 0 0>, rot=<0 0 -0.0132 -0.9999>}
 spine_05 {parent=spine_04, loc=<10.0146 0 0>, rot=<0 0 -0.0015 -1>}
 neck_01 {parent=spine_05, loc=<11.3138 0.1529 0>, rot=<0 0 -0.2003 -0.9797>}
 neck_02 {parent=neck_01, loc=<4.5627 0 0>, rot=<0 0 0.0504 -0.9987>}
 head {parent=neck_02, loc=<5.3914 0 0>, rot=<0 0 0.115 -0.9934>}
 attach_cape {parent=spine_05, loc=<9.017 -11.714 0>, rot=<0.0616 -0.7044 -0.0616 -0.7044>}
 attach_backpack {parent=spine_05, loc=<-0.123 -13.473 0>, rot=<-0.0093 -0.707 0.0093 -0.707>}
 clavicle_l {parent=spine_05, loc=<4.3575 -0.1466 -2.3223>, rot=<0.0465 -0.7289 -0.009 -0.6829>}
 upperarm_l {parent=clavicle_l, loc=<15.5111 0 0>, rot=<0.0226 -0.3234 -0.0088 -0.9459>}
 lowerarm_l {parent=upperarm_l, loc=<26.6514 0 0>, rot=<0.0391 -0.0099 -0.2453 -0.9686>}
 hand_l {parent=lowerarm_l, loc=<22.3491 0 0>, rot=<-0.7081 -0.0507 0.0975 -0.6975>}
 weapon_l {parent=hand_l, loc=<-0.1974 -3.4111 0.3811>, rot=<0 0 0 1>}
 thumb_01_l {parent=hand_l, loc=<1.3118 -0.3191 -2.459>, rot=<0.6546 -0.4042 0.0664 -0.6354>}
 thumb_02_l {parent=thumb_01_l, loc=<5.9336 -0.0448 0.0401>, rot=<-0.0037 0.0213 0.298 -0.9543>}
 thumb_03_l {parent=thumb_02_l, loc=<2.2223 0.3682 -0.0907>, rot=<0.0118 -0.013 0.0066 -0.9998>}
 index_metacarpal_l {parent=hand_l, loc=<1.6535 0.0183 -2.96>, rot=<0.0158 -0.0124 -0.0173 -0.9996>}
 index_01_l {parent=index_metacarpal_l, loc=<8.3008 0.0357 -0.0524>, rot=<-0.0044 0.0485 0.0924 -0.9945>}
 index_02_l {parent=index_01_l, loc=<3.3587 -0.3004 0.0452>, rot=<-0.002 0.0027 0.3225 -0.9466>}
 index_03_l {parent=index_02_l, loc=<2.5578 -0.123 -0.0252>, rot=<-0.002 0.0067 0.2558 -0.9667>}
 middle_metacarpal_l {parent=hand_l, loc=<2.1856 -0.1725 -0.5984>, rot=<0.006 0.0168 -0.0329 -0.9993>}
 middle_01_l {parent=middle_metacarpal_l, loc=<7.3275 -0.2447 0.0129>, rot=<-0.0208 0.0042 0.0858 -0.9961>}
 middle_02_l {parent=middle_01_l, loc=<4.0123 0.1401 0.1175>, rot=<-0.0218 0.0258 0.3512 -0.9357>}
 middle_03_l {parent=middle_02_l, loc=<2.6077 0.0534 0.1373>, rot=<0.0082 0.0188 0.1708 -0.9851>}
 ring_metacarpal_l {parent=hand_l, loc=<2.2074 -0.3625 1.3638>, rot=<-0.0284 0.0722 -0.0458 -0.9959>}
 ring_01_l {parent=ring_metacarpal_l, loc=<7.0295 -0.6197 -0.1446>, rot=<-0.0489 -0.0252 0.1892 -0.9804>}
 ring_02_l {parent=ring_01_l, loc=<3.6174 0.4647 -0.1027>, rot=<0 0 0.2806 -0.9598>}
 ring_03_l {parent=ring_02_l, loc=<2.4152 0.114 0.0199>, rot=<0 0 0.1189 -0.9929>}
 pinky_metacarpal_l {parent=hand_l, loc=<1.7616 -0.9753 3.3814>, rot=<-0.0397 0.0814 -0.069 -0.9935>}
 pinky_01_l {parent=pinky_metacarpal_l, loc=<7.1709 -0.4831 -0.1094>, rot=<-0.0654 -0.0035 0.2242 -0.9723>}
 pinky_02_l {parent=pinky_01_l, loc=<2.8274 0.0893 -0.1043>, rot=<0.0037 0.0011 0.2854 -0.9584>}
 pinky_03_l {parent=pinky_02_l, loc=<2.2046 0.0307 -0.0078>, rot=<-0.0002 0.0098 0.1445 -0.9895>}
 hand_attach_l {parent=hand_l, loc=<8.5 -3 0>, rot=<0.7071 0.7071 0 0>}
 deform_wrist_dn_base_l {parent=hand_l, loc=<0 -2.2342 0>, rot=<0 0 0 1>}
 deform_wrist_dn_l {parent=deform_wrist_dn_base_l, loc=<0 0 0>, rot=<0 0 0 1>}
 deform_wrist_up_base_l {parent=hand_l, loc=<0 2.2682 0>, rot=<0 0 0 -1>}
 deform_wrist_up_l {parent=deform_wrist_up_base_l, loc=<0 0 0>, rot=<0 0 0 1>}
 lowerarm_twist_02_l {parent=lowerarm_l, loc=<9.6199 0 0>, rot=<0 0 0 1>}
 lowerarm_twist_01_l {parent=lowerarm_l, loc=<17.8549 0 0>, rot=<0 0 0 1>}
 deform_elbow_in_base_l {parent=lowerarm_l, loc=<0.2112 5.259 0.8274>, rot=<0 0 0 1>}
 deform_elbow_in_l {parent=deform_elbow_in_base_l, loc=<0 0 0>, rot=<0 0 0 1>}
 deform_elbow_out_base_l {parent=lowerarm_l, loc=<0 -4 0>, rot=<0 0 0 1>}
 deform_elbow_out_l {parent=deform_elbow_out_base_l, loc=<0 0 0>, rot=<0 0 0 1>}
 upperarm_twist_01_l {parent=upperarm_l, loc=<0.0041 0 0>, rot=<0 0 0 1>}
 deform_upperarm_bk_base_l {parent=upperarm_twist_01_l, loc=<1.0813 -7.2193 -1.9453>, rot=<0 0 0 1>}
 deform_upperarm_bk_l {parent=deform_upperarm_bk_base_l, loc=<0 0 0>, rot=<0 0 0 1>}
 deform_upperarm_fr_base_l {parent=upperarm_twist_01_l, loc=<-0.3691 7.0506 -2.765>, rot=<0 0 0 1>}
 deform_upperarm_fr_l {parent=deform_upperarm_fr_base_l, loc=<0 0 0>, rot=<0 0 0 1>}
 deform_upperarm_up_base_l {parent=upperarm_twist_01_l, loc=<-3.2216 -1.1391 3.3311>, rot=<0 0 0 1>}
 deform_upperarm_up_l {parent=deform_upperarm_up_base_l, loc=<0 0 0>, rot=<0 0 0 1>}
 deform_upperarm_cape_up_l {parent=deform_upperarm_up_l, loc=<0 0 0>, rot=<0 0 0 1>}
 pad_r {parent=upperarm_twist_01_l, loc=<2.6967 -1.9865 5.1341>, rot=<0.6804 -0.0549 0.7284 -0.0588>}
 upperarm_twist_02_l {parent=upperarm_l, loc=<13.7941 0 0>, rot=<0 0 0 1>}
 clavicle_r {parent=spine_05, loc=<4.3575 -0.1466 2.3223>, rot=<0.7289 0.0465 -0.6829 0.009>}
 upperarm_r {parent=clavicle_r, loc=<-15.5111 0 0>, rot=<0.0226 -0.3234 -0.0088 -0.9459>}
 lowerarm_r {parent=upperarm_r, loc=<-26.6515 0 0>, rot=<0.0391 -0.0099 -0.2453 -0.9686>}
 hand_r {parent=lowerarm_r, loc=<-22.3492 0 0>, rot=<-0.7081 -0.0507 0.0975 -0.6975>}
 weapon_r {parent=hand_r, loc=<0.1975 3.4112 -0.3811>, rot=<0 0 0 -1>}
 thumb_01_r {parent=hand_r, loc=<-1.3119 0.3192 2.459>, rot=<0.6546 -0.4042 0.0664 -0.6354>}
 thumb_02_r {parent=thumb_01_r, loc=<-5.9335 0.0447 -0.0401>, rot=<-0.0037 0.0213 0.298 -0.9543>}
 thumb_03_r {parent=thumb_02_r, loc=<-2.2223 -0.3682 0.0907>, rot=<0.0118 -0.013 0.0066 -0.9998>}
 index_metacarpal_r {parent=hand_r, loc=<-1.6534 -0.0181 2.9601>, rot=<0.0158 -0.0124 -0.0173 -0.9996>}
 index_01_r {parent=index_metacarpal_r, loc=<-8.3008 -0.0357 0.0524>, rot=<-0.0044 0.0485 0.0924 -0.9945>}
 index_02_r {parent=index_01_r, loc=<-3.3589 0.3003 -0.0453>, rot=<-0.002 0.0027 0.3225 -0.9466>}
 index_03_r {parent=index_02_r, loc=<-2.5578 0.1231 0.0251>, rot=<-0.002 0.0067 0.2558 -0.9667>}
 middle_metacarpal_r {parent=hand_r, loc=<-2.1856 0.1726 0.5984>, rot=<0.006 0.0168 -0.0329 -0.9993>}
 middle_01_r {parent=middle_metacarpal_r, loc=<-7.3273 0.2447 -0.0129>, rot=<-0.0208 0.0042 0.0858 -0.9961>}
 middle_02_r {parent=middle_01_r, loc=<-4.0142 -0.1411 -0.07>, rot=<-0.0217 0.0259 0.3512 -0.9357>}
 middle_03_r {parent=middle_02_r, loc=<-2.6074 -0.0534 -0.1312>, rot=<0.0083 0.0188 0.1708 -0.9851>}
 ring_metacarpal_r {parent=hand_r, loc=<-2.2073 0.3627 -1.3638>, rot=<-0.0284 0.0722 -0.0458 -0.9959>}
 ring_01_r {parent=ring_metacarpal_r, loc=<-7.0297 0.6197 0.1446>, rot=<-0.0489 -0.0252 0.1892 -0.9804>}
 ring_02_r {parent=ring_01_r, loc=<-3.6171 -0.4648 0.1027>, rot=<0 0 0.2806 -0.9598>}
 ring_03_r {parent=ring_02_r, loc=<-2.4153 -0.1141 -0.0199>, rot=<0 0 0.1189 -0.9929>}
 pinky_metacarpal_r {parent=hand_r, loc=<-1.7617 0.9753 -3.3814>, rot=<-0.0397 0.0814 -0.069 -0.9935>}
 pinky_01_r {parent=pinky_metacarpal_r, loc=<-7.1707 0.4832 0.1094>, rot=<-0.0654 -0.0035 0.2242 -0.9723>}
 pinky_02_r {parent=pinky_01_r, loc=<-2.8274 -0.0892 0.1043>, rot=<0.0037 0.0011 0.2854 -0.9584>}
 pinky_03_r {parent=pinky_02_r, loc=<-2.2045 -0.031 0.0078>, rot=<-0.0002 0.0098 0.1445 -0.9895>}
 hand_attach_r {parent=hand_r, loc=<-8.5001 2.9999 0>, rot=<0 0 -0.7071 -0.7071>}
 deform_wrist_dn_base_r {parent=hand_r, loc=<-0.0004 2.2345 -0.0001>, rot=<0 0 0 -1>}
 deform_wrist_dn_r {parent=deform_wrist_dn_base_r, loc=<0 0 0>, rot=<0 0 0 1>}
 deform_wrist_up_base_r {parent=hand_r, loc=<0.0003 -2.2684 0.0001>, rot=<0 0 0 -1>}
 deform_wrist_up_r {parent=deform_wrist_up_base_r, loc=<0 0 0>, rot=<0 0 0 1>}
 lowerarm_twist_02_r {parent=lowerarm_r, loc=<-9.6199 0 0>, rot=<0 0 0 1>}
 lowerarm_twist_01_r {parent=lowerarm_r, loc=<-17.855 0 0>, rot=<0 0 0 1>}
 deform_elbow_in_base_r {parent=lowerarm_r, loc=<-0.211 -5.2591 -0.8275>, rot=<0 0 0 1>}
 deform_elbow_in_r {parent=deform_elbow_in_base_r, loc=<-0.0001 0 0.0002>, rot=<0 0 0 1>}
 deform_elbow_out_base_r {parent=lowerarm_r, loc=<0 4 -0.0001>, rot=<0 0 0 1>}
 deform_elbow_out_r {parent=deform_elbow_out_base_r, loc=<0 0 0>, rot=<0 0 0 1>}
 upperarm_twist_01_r {parent=upperarm_r, loc=<-0.0041 0 0>, rot=<0 0 0 1>}
 deform_upperarm_bk_base_r {parent=upperarm_twist_01_r, loc=<-1.0814 7.2192 1.9456>, rot=<0 0 0 1>}
 deform_upperarm_bk_r {parent=deform_upperarm_bk_base_r, loc=<0 0 0>, rot=<0 0 0 1>}
 deform_upperarm_fr_base_r {parent=upperarm_twist_01_r, loc=<0.3689 -7.0506 2.7652>, rot=<0 0 0 1>}
 deform_upperarm_fr_r {parent=deform_upperarm_fr_base_r, loc=<0 0 0>, rot=<0 0 0 1>}
 deform_upperarm_up_base_r {parent=upperarm_twist_01_r, loc=<3.2214 1.1391 -3.3308>, rot=<0 0 0 1>}
 deform_upperarm_up_r {parent=deform_upperarm_up_base_r, loc=<0 0 0>, rot=<0 0 0 -1>}
 deform_upperarm_cape_up_r {parent=deform_upperarm_up_r, loc=<0 0 0>, rot=<0 0 0 -1>}
 pad_l {parent=upperarm_twist_01_r, loc=<-2.6966 1.9865 -5.1342>, rot=<-0.0549 -0.6804 -0.0588 -0.7284>}
 upperarm_twist_02_r {parent=upperarm_r, loc=<-13.7942 0 0>, rot=<0 0 0 1>}
 MSTR_bandage {parent=upperarm_twist_02_r, loc=<-2.2148 6.446 1.2278>, rot=<-0.0053 0.1138 -0.7399 -0.663>}
 bandage {parent=MSTR_bandage, loc=<0 0 0>, rot=<0 0 0 1>}
 bandage_l_1 {parent=bandage, loc=<0.9801 -1.2083 1.5287>, rot=<0.0354 -0.1059 0.3255 -0.9389>}
 bandage_l_2 {parent=bandage_l_1, loc=<3.9085 0 0>, rot=<-0.0332 0.0301 -0.3923 -0.9187>}
 bandage_r_1 {parent=bandage, loc=<-0.4377 0.5035 -0.9914>, rot=<-0.0574 -0.0323 -0.2687 -0.961>}
 bandage_r_2 {parent=bandage_r_1, loc=<6.0782 0 0>, rot=<0.0103 0.0065 0.1922 -0.9813>}
 MSTR_mantle_1 {parent=spine_05, loc=<6.6601 -11.9334 0>, rot=<0 0 0.9959 -0.0905>}
 mantle_1 {parent=MSTR_mantle_1, loc=<0 0 0>, rot=<0 0 0 1>}
 mantle_2 {parent=mantle_1, loc=<10.2581 0 0>, rot=<0 0 0.0334 -0.9994>}
 mantle_3 {parent=mantle_2, loc=<8.0537 0 0>, rot=<0 0 0.0228 -0.9997>}
 mantle_4 {parent=mantle_3, loc=<8.8369 0 0>, rot=<0 0 0 1>}
 mantle_5 {parent=mantle_4, loc=<7.8639 0 0>, rot=<0 0 0 1>}
 MSTR_buckle {parent=spine_05, loc=<6.7167 9.835 -9.3576>, rot=<-0.4662 -0.4368 -0.2989 -0.7089>}
 buckle {parent=MSTR_buckle, loc=<0 0 0>, rot=<0 0 0 1>}
 deform_pec_aim_base_l {parent=spine_04, loc=<7.8732 12.6388 -2.4123>, rot=<0.0102 -0.4602 0.3979 -0.7936>}
 deform_pec_aim_l {parent=deform_pec_aim_base_l, loc=<0 0 0>, rot=<0 0 0 1>}
 deform_pec_base_l {parent=deform_pec_aim_l, loc=<10.766 0 0>, rot=<0 0 0 1>}
 deform_pec_l {parent=deform_pec_base_l, loc=<0.055 0 0>, rot=<0 0 0 1>}
 deform_pec_aim_base_r {parent=spine_04, loc=<7.8727 12.6388 2.4123>, rot=<-0.4602 -0.0102 0.7936 0.3979>}
 deform_pec_aim_r {parent=deform_pec_aim_base_r, loc=<0 0 0>, rot=<0 0 0 1>}
 deform_pec_base_r {parent=deform_pec_aim_r, loc=<-10.7662 -0.0006 -0.0007>, rot=<0 0 0 1>}
 deform_pec_r {parent=deform_pec_base_r, loc=<0 0 0>, rot=<0 0 0 1>}
 chest_belt {parent=spine_04, loc=<6.3097 14.7199 -3.0945>, rot=<0 0 0.088 -0.9961>}
 button_01_mid {parent=spine_02, loc=<0.0864 12.9763 0.2412>, rot=<0 0 -0.0495 -0.9988>, scale=<1.0001 1 1>}
 MSTR_rope_l_1 {parent=spine_01, loc=<2.9069 8.5204 -12.9669>, rot=<-0.0513 0.4604 0.8862 0.0082>, scale=<0.9999 1 1>}
 rope_l_1 {parent=MSTR_rope_l_1, loc=<0 0 0>, rot=<0 0 0 1>}
 rope_l_2 {parent=rope_l_1, loc=<3.7072 0 0>, rot=<-0.008 -0.0016 0.0843 -0.9964>}
 rope_l_3 {parent=rope_l_2, loc=<8.1274 0 0>, rot=<-0.002 -0.0007 0.0148 -0.9999>}
 rope_l_4 {parent=rope_l_3, loc=<8.0333 0 0>, rot=<-0.035 -0.0038 -0.0028 -0.9994>}
 rope_l_5 {parent=rope_l_4, loc=<12.4496 0 0>, rot=<0.0155 0.0078 -0.0039 -0.9998>}
 MSTR_rope_mid {parent=spine_01, loc=<1.7824 13.2344 0>, rot=<0 0 -0.0022 -1>, scale=<0.9999 1 1>}
 rope_mid {parent=MSTR_rope_mid, loc=<0 0 0>, rot=<0 0 0 1>}
 MSTR_rope_r_1 {parent=spine_01, loc=<2.9067 8.5204 12.9669>, rot=<0.0513 -0.4604 0.8862 0.0082>, scale=<0.9999 1 1>}
 rope_r_1 {parent=MSTR_rope_r_1, loc=<0 0 0>, rot=<0 0 0 1>}
 thigh_l {parent=pelvis, loc=<-2.5693 -0.0109 -7.7501>, rot=<0.0012 0.0572 0.0155 -0.9982>}
 calf_l {parent=thigh_l, loc=<-39.0484 0 0>, rot=<0.0006 -0.0161 -0.0391 -0.9991>}
 foot_l {parent=calf_l, loc=<-42.561 0.1077 0.4292>, rot=<0.0418 -0.0455 0.0343 -0.9975>}
 ball_l {parent=foot_l, loc=<-6.1599 13.8264 0.6806>, rot=<0.0001 -0.0001 -0.7063 -0.7079>}
 deform_heel_l_MSTR {parent=foot_l, loc=<0.9168 -1.1896 -0.0588>, rot=<0.7688 -0.6394 0 0>}
 defom_heel_l {parent=deform_heel_l_MSTR, loc=<0 0 0>, rot=<0 0 0 1>}
 MSTR_deform_foot_belt_l {parent=foot_l, loc=<2.5357 6.4058 0.5965>, rot=<0.9147 0.4042 0 0>}
 deform_foot_belt_l {parent=MSTR_deform_foot_belt_l, loc=<0 0 0>, rot=<0 0 0 1>}
 calf_twist_02_l {parent=calf_l, loc=<-18.8064 -0.0967 0.0397>, rot=<0.0237 0.0095 -0.0035 -0.9997>}
 calf_twist_01_l {parent=calf_l, loc=<-35.813 -0.0967 0.0397>, rot=<0.0237 0.0095 -0.0035 -0.9997>}
 deform_knee_in_base_l {parent=calf_l, loc=<1.298 -5.2957 -0.0496>, rot=<0 0 0 -1>}
 deform_knee_in_l {parent=deform_knee_in_base_l, loc=<0 0 0>, rot=<0 0 0 1>}
 deform_knee_out_base_l {parent=calf_l, loc=<0 4 0>, rot=<0 0 0 -1>}
 deform_knee_out_l {parent=deform_knee_out_base_l, loc=<0 0 0>, rot=<0 0 0 1>}
 fur_l {parent=calf_l, loc=<-8.3074 -8.8861 0>, rot=<0.7445 -0.6676 0 0>}
 thigh_twist_01_l {parent=thigh_l, loc=<-21.6303 0 0>, rot=<0.0013 0 0 -1>}
 deform_groin_base_l {parent=thigh_l, loc=<-2.1908 8.1142 -1.7442>, rot=<0 0 0 -1>}
 deform_groin_l {parent=deform_groin_base_l, loc=<0 0 0>, rot=<0 0 0 1>}
 thigh_r {parent=pelvis, loc=<-2.5691 -0.0109 7.7501>, rot=<0.0572 -0.0012 0.9982 0.0155>}
 calf_r {parent=thigh_r, loc=<39.0487 0 0>, rot=<0.0006 -0.0161 -0.0391 -0.9991>}
 foot_r {parent=calf_r, loc=<42.5612 -0.1077 -0.4292>, rot=<0.0418 -0.0455 0.0343 -0.9975>}
 ball_r {parent=foot_r, loc=<6.1598 -13.8264 -0.6806>, rot=<0.0001 -0.0001 -0.7063 -0.7079>}
 deform_heel_r_MSTR {parent=foot_r, loc=<-0.917 1.1896 0.0587>, rot=<0 0 -0.7688 -0.6394>}
 defom_heel_r {parent=deform_heel_r_MSTR, loc=<0 0 0>, rot=<0 0 0 1>}
 MSTR_deform_foot_belt_r {parent=foot_r, loc=<-2.5359 -6.4058 -0.5965>, rot=<0 0 0.9147 -0.4042>}
 deform_foot_belt_r {parent=MSTR_deform_foot_belt_r, loc=<0 0 0>, rot=<0 0 0 1>}
 calf_twist_02_r {parent=calf_r, loc=<18.8064 0.0967 -0.0397>, rot=<0.0237 0.0095 -0.0035 -0.9997>}
 calf_twist_01_r {parent=calf_r, loc=<35.8132 0.0967 -0.0397>, rot=<0.0237 0.0095 -0.0035 -0.9997>}
 deform_knee_in_base_r {parent=calf_r, loc=<-1.2981 5.2958 0.0496>, rot=<0 0 0 1>}
 deform_knee_in_r {parent=deform_knee_in_base_r, loc=<0 0 0>, rot=<0 0 0 1>}
 deform_knee_out_base_r {parent=calf_r, loc=<-0.0002 -4 0>, rot=<0 0 0 -1>}
 deform_knee_out_r {parent=deform_knee_out_base_r, loc=<0 0 0>, rot=<0 0 0 1>}
 fur_r {parent=calf_r, loc=<8.3074 8.8861 0>, rot=<0 0 -0.7445 -0.6676>}
 thigh_twist_01_r {parent=thigh_r, loc=<21.6303 0 0>, rot=<0.0013 0 0 -1>}
 deform_groin_base_r {parent=thigh_r, loc=<2.1908 -8.1142 1.7442>, rot=<0 0 0 -1>}
 deform_groin_r {parent=deform_groin_base_r, loc=<0 0 0>, rot=<0 0 0 1>}
 deform_glute_base_l {parent=pelvis, loc=<-8.8064 -8.8222 -8.4287>, rot=<0.0275 0.0501 -0.4472 -0.8926>}
 deform_glute_l {parent=deform_glute_base_l, loc=<0 0 0>, rot=<0 0 0 1>}
 deform_glute_base_r {parent=pelvis, loc=<-8.8061 -8.8222 8.4287>, rot=<0.0501 -0.0275 0.8926 -0.4472>}
 deform_glute_r {parent=deform_glute_base_r, loc=<0 0 0>, rot=<0 0 0 1>}
 MSTR_skirt_A_r {parent=pelvis, loc=<1.0123 11.0203 6.0328>, rot=<0.0841 -0.1469 0.9831 0.0703>}
 skirt_A_r_1 {parent=MSTR_skirt_A_r, loc=<0 0 0>, rot=<0 0 0.0155 -0.9999>}
 skirt_A_r_2 {parent=skirt_A_r_1, loc=<10.3805 0 0>, rot=<0.0006 0.0243 0.0247 -0.9994>}
 skirt_A_r_1_belt {parent=skirt_A_r_1, loc=<0.0291 -0.0632 0>, rot=<0 0 0 1>}
 MSTR_skirt_B_r {parent=pelvis, loc=<1.9962 2.7835 13.393>, rot=<0.1305 -0.6314 0.7539 0.1265>}
 skirt_B_r_1 {parent=MSTR_skirt_B_r, loc=<0 0 0>, rot=<0 0 0 1>}
 skirt_B_r_2 {parent=skirt_B_r_1, loc=<9.7236 0 0>, rot=<0.002 0.0242 -0.0314 -0.9992>}
 MSTR_skirt_C_r {parent=pelvis, loc=<2.2915 -7.2227 11.9583>, rot=<0.0418 -0.9294 0.3635 0.0478>}
 skirt_C_r_1 {parent=MSTR_skirt_C_r, loc=<0 0 0>, rot=<0 0 0 1>}
 skirt_C_r_2 {parent=skirt_C_r_1, loc=<9.7236 0 0>, rot=<0.0009 0.0243 0.0129 -0.9996>}
 MSTR_skirt_mid {parent=pelvis, loc=<2.2915 -11.8488 0>, rot=<0.0462 -0.9988 -0.0147 -0.0001>}
 skirt_mid_1 {parent=MSTR_skirt_mid, loc=<0 0 0>, rot=<0 0 0 1>}
 skirt_mid_2 {parent=skirt_mid_1, loc=<9.7236 0 0>, rot=<0.0006 0 -0.0217 -0.9998>}
 MSTR_skirt_A_l {parent=pelvis, loc=<1.0123 11.0203 -6.0328>, rot=<-0.0841 0.1469 0.9831 0.0703>}
 skirt_A_l_1 {parent=MSTR_skirt_A_l, loc=<0 0 0>, rot=<0 0 0.0155 -0.9999>}
 skirt_A_l_2 {parent=skirt_A_l_1, loc=<10.3805 0 0>, rot=<-0.0006 -0.0243 0.0247 -0.9994>}
 skirt_A_l_1_belt {parent=skirt_A_l_1, loc=<0.0291 -0.0632 0>, rot=<0 0 0 1>}
 MSTR_skirt_B_l {parent=pelvis, loc=<1.9962 2.7835 -13.393>, rot=<-0.1305 0.6314 0.7539 0.1265>}
 skirt_B_l_1 {parent=MSTR_skirt_B_l, loc=<0 0 0>, rot=<0 0 0 1>}
 skirt_B_l_2 {parent=skirt_B_l_1, loc=<9.7236 0 0>, rot=<-0.002 -0.0242 -0.0314 -0.9992>}
 MSTR_skirt_C_l {parent=pelvis, loc=<2.2915 -7.2227 -11.9583>, rot=<0.0418 -0.9294 -0.3635 -0.0478>}
 skirt_C_l_1 {parent=MSTR_skirt_C_l, loc=<0 0 0>, rot=<0 0 0 1>}
 skirt_C_l_2 {parent=skirt_C_l_1, loc=<9.7236 0.0001 -0.0001>, rot=<-0.0009 -0.0243 0.0129 -0.9996>}
 MSTR_skirt_D_l {parent=pelvis, loc=<1.6555 8.2731 -11.0871>, rot=<-0.0957 0.2946 0.9481 0.0718>}
 skirt_D_l_1 {parent=MSTR_skirt_D_l, loc=<0 0 0>, rot=<0 0 0.0155 -0.9999>}
 skirt_D_l_2 {parent=skirt_D_l_1, loc=<10.3805 0 0>, rot=<-0.0006 -0.0243 0.0247 -0.9994>}
 MSTR_belt_mid {parent=pelvis, loc=<2.2519 13.527 0>, rot=<0 0 0.0318 -0.9995>}
 belt_mid {parent=MSTR_belt_mid, loc=<0 0 0>, rot=<0 0 0 1>}
 MSTR_skirt_D_r {parent=pelvis, loc=<1.6555 8.2731 11.0871>, rot=<0.0957 -0.2946 0.9481 0.0718>}
 skirt_D_r_1 {parent=MSTR_skirt_D_r, loc=<0 0 0>, rot=<0 0 0.0155 -0.9999>}
 skirt_D_r_2 {parent=skirt_D_r_1, loc=<10.3804 0 0>, rot=<0.0006 0.0243 0.0247 -0.9994>}
 attach {parent=root, loc=<0 0 0>, rot=<0 0 0 1>}
 ik_hand_root {parent=root, loc=<0 0 0>, rot=<0 0 0 1>}
 ik_hand_gun {parent=ik_hand_root, loc=<-51.9729 4.0884 102.6935>, rot=<0.6084 0.1793 0.4613 -0.6205>}
 ik_hand_r {parent=ik_hand_gun, loc=<0 0 0>, rot=<0 0 0 1>}
 ik_hand_l {parent=ik_hand_gun, loc=<53.0286 82.1782 35.2051>, rot=<-0.6205 -0.4613 0.1793 -0.6084>}
 ik_foot_root {parent=root, loc=<0 0 0>, rot=<0 0 0 1>}
 ik_foot_l {parent=ik_foot_root, loc=<15.2608 -4.4176 9.4433>, rot=<0.0357 0.7045 0.0284 -0.7082>}
 ik_foot_r {parent=ik_foot_root, loc=<-15.2609 -4.4176 9.4431>, rot=<0.7082 0.0284 -0.7045 0.0357>}
 root_fp {parent=root, loc=<0 0 0>, rot=<0 0 0 1>}
 attach_fp {parent=root_fp, loc=<0 0 0>, rot=<0 0 0 1>}
 camera_root {parent=root_fp, loc=<0 0 0>, rot=<0 0 0 1>}
 camera {parent=camera_root, loc=<0 0 0>, rot=<0 0 0 1>}
 attach_emote_01 {parent=root, loc=<0 0 0>, rot=<0 0 0 1>}
 attach_emote_02 {parent=root, loc=<0 0 0>, rot=<0 0 0 1>}
 attach_emote_03 {parent=root, loc=<0 0 0>, rot=<0 0 0 1>}
 warp_point_01 {parent=root, loc=<0 0 0>, rot=<0 0 0 1>}
 warp_point_02 {parent=root, loc=<0 0 0>, rot=<0 0 0 1>}
 warp_point_03 {parent=root, loc=<0 0 0>, rot=<0 0 0 1>}
]
