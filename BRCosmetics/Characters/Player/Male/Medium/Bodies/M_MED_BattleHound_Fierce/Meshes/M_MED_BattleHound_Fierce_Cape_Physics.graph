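# Physics asset for the cape of the male medium BattleHound Fierce body cosmetic, defining its simulation bodies and joints.
# Five box bodies form the mantle chain, hung from a small capsule on MSTR_mantle_1 and linked by twist-locked constraints whose first swing limit widens toward the tip.
# Unconstrained capsules on the pelvis, skirt and upper spine approximate the torso, likely as collision for the cape.
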PhysicsAsset /BRCosmetics/Characters/Player/Male/Medium/Bodies/M_MED_BattleHound_Fierce/Meshes/M_MED_BattleHound_Fierce_Cape_Physics
bodies {
 MSTR_mantle_1 [{shapeType=Capsule, length=1, radius=1, shapeName=MSTR_mantle_1_capsule}]
 mantle_1 [{shapeType=Box, extentX=9, extentY=4.25, extentZ=38.5, shapeName=mantle_1_box, center=<5.2637 0.7289 0>}]
 mantle_2 [{shapeType=Box, extentX=8, extentY=4.75, extentZ=38.5, shapeName=mantle_2_box, center=<3.7943 0.276 0>, rotation=<rot 0 -0.0002 0>}]
 mantle_3 [{shapeType=Box, extentX=9, extentY=3.5, extentZ=36.75, shapeName=mantle_3_box, center=<4.0935 -0.2879 0>}]
 mantle_4 [{shapeType=Box, extentX=7, extentY=3, extentZ=29.25, shapeName=mantle_4_box, center=<3.5158 -0.2767 0>}]
 mantle_5 [{shapeType=Box, extentX=7, extentY=3, extentZ=19.5, shapeName=mantle_5_box, center=<3.5697 -0.474 0>}]
 pelvis [{shapeType=Capsule, length=10.3504, radius=20.25, shapeName=pelvis_capsule, center=<10.0805 4.5232 0>}]
 skirt_mid_2 [{shapeType=Capsule, length=14.0625, radius=9.5565, shapeName=skirt_mid_2_capsule, center=<3.0701 5.8042 0>}]
 spine_03 [{shapeType=Capsule, length=18.1402, radius=15.8989, shapeName=spine_03_capsule, center=<3.0841 1.209 0>}]
 spine_05 [{shapeType=Capsule, length=22.2267, radius=14.6483, shapeName=spine_05_capsule, center=<-1.8727 -0.011 0>}]
}
constraints [
 {bone1Name=mantle_1, bone2Name=mantle_2, swing1LimitDegrees=23.9501, swing2LimitDegrees=8, twistLimitDegrees=0, twistMotion=Locked}
 {bone1Name=mantle_2, bone2Name=mantle_3, swing1LimitDegrees=29.9376, swing2LimitDegrees=8, twistLimitDegrees=0, twistMotion=Locked}
 {bone1Name=mantle_3, bone2Name=mantle_4, swing1LimitDegrees=35, swing2LimitDegrees=8, twistLimitDegrees=0, twistMotion=Locked}
 {bone1Name=mantle_4, bone2Name=mantle_5, swing1LimitDegrees=40, swing2LimitDegrees=8, twistLimitDegrees=0, twistMotion=Locked}
 {bone1Name=MSTR_mantle_1, bone2Name=mantle_1, swing1LimitDegrees=14.256, swing2LimitDegrees=8, twistLimitDegrees=0, twistMotion=Locked}
]
