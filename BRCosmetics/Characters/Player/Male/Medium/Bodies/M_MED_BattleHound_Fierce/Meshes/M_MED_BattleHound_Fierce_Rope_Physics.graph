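# Physics asset for the left-side rope on the BattleHound Fierce medium male player body cosmetic.
# Five capsule bodies, rope_l_1 to rope_l_5, form a chain hanging from MSTR_rope_l_1, each joint constrained with moderate swing and tight twist limits.
# Unconstrained pelvis, left thigh and left skirt bodies, mostly capsules plus one box, supply collision shapes around the rope.
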PhysicsAsset /BRCosmetics/Characters/Player/Male/Medium/Bodies/M_MED_BattleHound_Fierce/Meshes/M_MED_BattleHound_Fierce_Rope_Physics
bodies {
 MSTR_rope_l_1 [{shapeType=Capsule, length=1, radius=1, shapeName=MSTR_rope_l_1_capsule}]
 MSTR_skirt_C_l [{
   shapeType=Capsule
   length=5.5
   radius=12.875
   shapeName=MSTR_skirt_C_l_capsule
   center=<13.4035 0.5465 -7.491>
   rotation=<rot 0 -89.9999 89.9999>
  }]
 pelvis [
  {
   shapeType=Capsule
   length=21.4493
   radius=22.4737
   shapeName=pelvis_capsule
   center=<16.2923 -2.8136 8.5639>
   rotation=<rot 0.8501 -4.9849 91.0707>
  }
  {shapeType=Capsule, length=6.25, radius=11.625, shapeName=pelvis_capsule, center=<0 -0.3917 0>}
 ]
 rope_l_1 [{shapeType=Capsule, length=0.6829, radius=2.3, shapeName=rope_l_1_capsule, center=<3.1175 0 0>, rotation=<rot 90 0 0>}]
 rope_l_2 [{
   shapeType=Capsule
   length=1.5446
   radius=2.05
   shapeName=rope_l_2_capsule
   center=<5.1235 -0.0888 0.5601>
   rotation=<rot 90 -163.4419 -163.4418>
  }]
 rope_l_3 [{shapeType=Capsule, length=0.4994, radius=3.8, shapeName=rope_l_3_capsule, center=<4.6788 0 0>, rotation=<rot 90 0 0>}]
 rope_l_4 [{shapeType=Capsule, length=2.9447, radius=4.05, shapeName=rope_l_4_capsule, center=<6.0492 0.3316 0>, rotation=<rot 90 0 0>}]
 rope_l_5 [{
   shapeType=Capsule
   length=3.2047
   radius=4.425
   shapeName=rope_l_5_capsule
   center=<5.0855 -0.3201 -0.1299>
   rotation=<rot 87.1876 -0.0019 5.6232>
  }]
 skirt_B_l_2 [{shapeType=Capsule, length=8.3233, radius=5.4528, shapeName=skirt_B_l_2_capsule, center=<4.1955 5.0758 -0.2102>, rotation=<rot 0 90 -87.1875>}]
 skirt_C_l_2 [{
   shapeType=Box
   extentX=2.0253
   extentY=12.9359
   extentZ=18.5422
   shapeName=skirt_C_l_2_box
   center=<5.8172 3.373 0.7763>
   rotation=<rot 14.4828 -96.7301 -4.2054>
  }]
 thigh_l [{
   shapeType=Capsule
   length=22.1263
   radius=14.8253
   shapeName=thigh_l_capsule
   center=<-22.5289 0.8744 5.6679>
   rotation=<rot 87.1875 176.5601 179.999>
  }]
}
constraints [
 {bone1Name=rope_l_1, bone2Name=MSTR_rope_l_1, swing1LimitDegrees=30, swing2LimitDegrees=30, twistLimitDegrees=5}
 {bone1Name=rope_l_2, bone2Name=rope_l_1, swing1LimitDegrees=30, swing2LimitDegrees=30, twistLimitDegrees=5}
 {bone1Name=rope_l_3, bone2Name=rope_l_2, swing1LimitDegrees=30, swing2LimitDegrees=30, twistLimitDegrees=5}
 {bone1Name=rope_l_4, bone2Name=rope_l_3, swing1LimitDegrees=30, swing2LimitDegrees=30, twistLimitDegrees=5}
 {bone1Name=rope_l_5, bone2Name=rope_l_4, swing1LimitDegrees=30, swing2LimitDegrees=30, twistLimitDegrees=5}
]
