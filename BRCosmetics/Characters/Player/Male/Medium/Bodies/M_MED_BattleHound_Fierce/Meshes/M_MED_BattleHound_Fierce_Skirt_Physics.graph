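# Physics asset for the skirt of the male medium BattleHound Fierce character body.
# Mid, left C and right C chains each link a small master capsule to two box panels through angular limits, tightly restricting twist.
# Small B-side tip capsules attach to the lower C panels with free swing and twist.
# Pelvis and thigh capsules have no constraints, providing body shapes alongside the panels, as does an unlinked right B master capsule.
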PhysicsAsset /BRCosmetics/Characters/Player/Male/Medium/Bodies/M_MED_BattleHound_Fierce/Meshes/M_MED_BattleHound_Fierce_Skirt_Physics
bodies {
 MSTR_skirt_B_r [{
   shapeType=Capsule
   length=2.5191
   radius=1.0426
   shapeName=MSTR_skirt_B_r_capsule
   center=<-1.7847 -1.0338 -2.1164>
   rotation=<rot 0 -80.2108 -89.1637>
  }]
 MSTR_skirt_C_l [{shapeType=Capsule, length=0.505, radius=0.505, shapeName=MSTR_skirt_C_l_capsule, rotation=<rot 0 -90 0>}]
 MSTR_skirt_C_r [{shapeType=Capsule, length=0.505, radius=0.505, shapeName=MSTR_skirt_C_r_capsule, rotation=<rot 0 -90 0>}]
 MSTR_skirt_mid [{shapeType=Capsule, length=1, radius=1, shapeName=MSTR_skirt_mid_capsule}]
 pelvis [{shapeType=Capsule, length=2.4035, radius=12.4498, shapeName=pelvis_capsule, center=<0 0.3834 0>, rotation=<rot 0 179.9995 -180>}]
 skirt_B_l_2 [{shapeType=Capsule, length=1, radius=1, shapeName=skirt_B_l_2_capsule}]
 skirt_B_r_2 [{shapeType=Capsule, length=1, radius=1, shapeName=skirt_B_r_2_capsule}]
 skirt_C_l_1 [{shapeType=Box, extentX=9, extentY=2.75, extentZ=13.25, shapeName=skirt_C_l_1_box, center=<4.5629 0 0>}]
 skirt_C_l_2 [{shapeType=Box, extentX=10, extentY=3, extentZ=16.75, shapeName=skirt_C_l_2_box, center=<5.1807 0 0>}]
 skirt_C_r_1 [{shapeType=Box, extentX=10.25, extentY=3.5, extentZ=14, shapeName=skirt_C_l_1_box, center=<5.2831 0 0>}]
 skirt_C_r_2 [{shapeType=Box, extentX=10, extentY=6.25, extentZ=18, shapeName=skirt_C_l_2_box, center=<5.1807 0 -1.2762>}]
 skirt_mid_1 [{shapeType=Box, extentX=9, extentY=1.75, extentZ=17.75, shapeName=skirt_mid_1_box, center=<4.5629 0 0>}]
 skirt_mid_2 [{shapeType=Box, extentX=10, extentY=4.25, extentZ=12.25, shapeName=skirt_mid_2_box, center=<5.1807 -0.3493 0>}]
 thigh_l [{
   shapeType=Capsule
   length=20.4705
   radius=10.692
   shapeName=thigh_l_capsule
   center=<-11.9804 2.1218 -0.1986>
   rotation=<rot 84.375 179.9998 179.9999>
  }]
 thigh_r [{shapeType=Capsule, length=20.6625, radius=10.5, shapeName=thigh_r_capsule, center=<12 -2 0>, rotation=<rot 84.375 179.9999 179.9999>}]
}
constraints [
 {bone1Name=skirt_mid_2, bone2Name=skirt_mid_1, swing1LimitDegrees=30, swing2LimitDegrees=30, twistLimitDegrees=1}
 {bone1Name=skirt_C_l_2, bone2Name=skirt_C_l_1, swing1LimitDegrees=6, swing2LimitDegrees=4, twistLimitDegrees=5}
 {bone1Name=skirt_C_l_1, bone2Name=MSTR_skirt_C_l, swing1LimitDegrees=5, swing2LimitDegrees=50, twistLimitDegrees=2}
 {bone1Name=skirt_mid_1, bone2Name=MSTR_skirt_mid, swing1LimitDegrees=30, swing2LimitDegrees=30, twistLimitDegrees=5}
 {
  bone1Name=skirt_B_l_2
  bone2Name=skirt_C_l_2
  swing1LimitDegrees=45
  swing1Motion=Free
  swing2LimitDegrees=45
  swing2Motion=Free
  twistLimitDegrees=45
  twistMotion=Free
 }
 {bone1Name=MSTR_skirt_C_r, bone2Name=skirt_C_r_1, swing1LimitDegrees=5, swing2LimitDegrees=50, twistLimitDegrees=2}
 {bone1Name=skirt_C_r_1, bone2Name=skirt_C_r_2, swing1LimitDegrees=10, swing2LimitDegrees=4, twistLimitDegrees=5}
 {
  bone1Name=skirt_B_r_2
  bone2Name=skirt_C_r_2
  swing1LimitDegrees=45
  swing1Motion=Free
  swing2LimitDegrees=45
  swing2Motion=Free
  twistLimitDegrees=45
  twistMotion=Free
 }
]
