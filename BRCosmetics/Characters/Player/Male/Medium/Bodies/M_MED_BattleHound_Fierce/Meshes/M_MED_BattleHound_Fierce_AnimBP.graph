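# Post-process style AnimBP for the BattleHound Fierce character body, copying the parent mesh pose and adding secondary motion.
# A Control Rig feeds chained RigidBody sims for skirt, bandage and rope (separate front-end asset).
# Cape blends between Control Rig and physics by backpack presence, or by socket distance for BattleHoundFire back bling, tilting while skydiving.
# When bRigidBodyEnabled is false, output passes through the Altskin Control Rig.
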AnimBlueprint /BRCosmetics/Characters/Player/Male/Medium/Bodies/M_MED_BattleHound_Fierce/Meshes/M_MED_BattleHound_Fierce_AnimBP
target_skeleton=&/M_MED_BattleHound_Fierce_Skeleton
graphs [
 AnimGraph {nodes [
   n1 = Control Rig!(Source=Use cached pose 'Main'(), Alpha, LOD:int=Get CurrentLOD()) -> (Pose) {
    alphaBoolBlend {blendOption=Linear}
    controlRigAssetReference {blueprintRigClass=&/M_MED_BattleHound_Fierce_CtrlRig_Altskin.M_MED_BattleHound_Fierce_CtrlRig_Altskin_C}
   }
   Output Pose(Result=Blend Poses by bool(BlendPose_0=Use cached pose 'Main'(), BlendPose_1=@n1, bActiveValue=Get bRigidBodyEnabled()) {blendType=Linear}) {layerGroup=None}
   n2 = Control Rig!(Source=Copy Pose From Mesh() {bUseAttachedParent=true}, Alpha, isMoving:double=Get isMoving() -> (isMoving:double)) -> (Pose) {
    alphaBoolBlend {blendOption=Linear}
    controlRigAssetReference {blueprintRigClass=&/M_MED_BattleHound_Fierce_CtrlRig.M_MED_BattleHound_Fierce_CtrlRig_C}
   }
   n3 = "Transform (Modify) Bone - Bone: attach_cape"!(
    ComponentPose=Local To Component(LocalPose=@n2)
    bAlphaBoolEnabled:bool=Get IsBattleHoundFireCape() -> (IsBattleHoundFireCape:bool)
    Translation="0,-1,0"
    Rotation="0,0,0"
    Scale="1,1,1"
   ) -> (Pose) {
    alphaInputType=Bool
    rotationMode=BMM_Additive
    rotationSpace=BCS_BoneSpace
    translationMode=BMM_Additive
    translationSpace=BCS_BoneSpace
    alphaBoolBlend {blendOption=Linear}
    boneToModify {boneName=attach_cape}
    translation {y=-1}
   }
   n4 = RigidBody!(ComponentPose=@n3, Alpha, OverrideWorldGravity:Vector="0,0,0", ExternalForce) -> (Pose) {
    bOverrideWorldGravity=true
    overridePhysicsAsset=&/M_MED_BattleHound_Fierce_Skirt_Physics
    simulationSpace=BaseBoneSpace
    alphaBoolBlend {blendOption=Linear}
    baseBoneRef {boneName=pelvis}
    componentAppliedLinearAccClamp=<3000 3000 3000>
    simSpaceSettings {
     maxAngularAcceleration=1
     maxAngularVelocity=3
     maxLinearAcceleration=50
     maxLinearVelocity=500
     worldAlpha=1
     externalAngularVelocity {z=0.1}
     externalLinearDragV=<0.2 0.2>
    }
   }
   n5 = RigidBody(ComponentPose=@n4, OverrideWorldGravity="0,0,0") {
    bOverrideWorldGravity=true
    overridePhysicsAsset=&/M_MED_BattleHound_Fierce_Bandage_Physics
    simulationSpace=BaseBoneSpace
    alphaBoolBlend {blendOption=Linear}
    baseBoneRef {boneName=upperarm_twist_02_r}
    componentAppliedLinearAccClamp=<3000 3000 3000>
    simSpaceSettings {
     maxAngularAcceleration=4
     maxAngularVelocity=3
     maxLinearAcceleration=50
     maxLinearVelocity=100
     worldAlpha=1
     externalAngularVelocity {z=0.1}
     externalLinearDragV=<0.5 0.5>
    }
   }
   n6 = RigidBody!(ComponentPose=@n5, bAlphaBoolEnabled:bool=NOT Boolean(A=Get bIsInFrontEnd()), OverrideWorldGravity="0,0,0", ExternalForce) -> (Pose) {
    alphaInputType=Bool
    bOverrideWorldGravity=true
    overridePhysicsAsset=&/M_MED_BattleHound_Fierce_Rope_Physics
    simulationSpace=BaseBoneSpace
    alphaBoolBlend {blendOption=Linear}
    baseBoneRef {boneName=thigh_l}
    componentAppliedLinearAccClamp=<3000 3000 3000>
    simSpaceSettings {
     maxAngularAcceleration=20
     maxAngularVelocity=4
     maxLinearAcceleration=100
     maxLinearVelocity=100
     worldAlpha=1
     externalAngularVelocity {z=0.1}
     externalLinearDragV=<0.2 0.2>
    }
   }
   n7 = RigidBody(ComponentPose=@n6, bAlphaBoolEnabled=Get bIsInFrontEnd(), OverrideWorldGravity="0,0,0") {
    alphaInputType=Bool
    bOverrideWorldGravity=true
    overridePhysicsAsset=&/M_MED_BattleHound_Fierce_Rope_FN_Physics
    simulationSpace=BaseBoneSpace
    alphaBoolBlend {blendOption=Linear}
    baseBoneRef {boneName=thigh_l}
    componentAppliedLinearAccClamp=<3000 3000 3000>
    simSpaceSettings {
     maxAngularAcceleration=20
     maxAngularVelocity=4
     maxLinearAcceleration=100
     maxLinearVelocity=100
     worldAlpha=1
     externalAngularVelocity {z=0.1}
     externalLinearDragV=<0.2 0.2>
    }
   }
   Save cached pose 'MainCache'(Pose=Component To Local(ComponentPose=@n7)) {cacheName=MainCache}
   n8 = Control Rig!(Source=Use cached pose 'MainCache'(), Alpha=Get NoCapeDynamics_Alpha() -> (NoCapeDynamics_Alpha:double)) -> (Pose) {
    alphaBoolBlend {blendOption=Linear}
    controlRigAssetReference {blueprintRigClass=&/M_MED_BattleHound_Fierce_Cape_CtrlRig.M_MED_BattleHound_Fierce_Cape_CtrlRig_C}
   }
   n9 = "Transform (Modify) Bone - Bone: attach_cape"(
    ComponentPose=Local To Component(LocalPose=@n8)
    bAlphaBoolEnabled=Get RotateCapeUp() -> (RotateCapeUp:bool)
    Translation="0,0,0"
    Rotation="0,0,18"
    Scale="1,1,1"
   ) {
    alphaInputType=Bool
    rotationMode=BMM_Additive
    translationMode=BMM_Additive
    alphaBoolBlend {blendOption=Linear}
    boneToModify {boneName=attach_cape}
    rotation {roll=18}
   }
   n10 = Get CapeDynamics_Alpha() -> (CapeDynamics_Alpha:double)
   n11 = "Transform (Modify) Bone - Bone: MSTR_mantle_1"(ComponentPose=@n9, Alpha=@n10, Translation="0,6,0", Rotation="0,0,0", Scale="1,1,1") {
    translationMode=BMM_Additive
    translationSpace=BCS_BoneSpace
    alphaBoolBlend {blendOption=Linear}
    boneToModify {boneName=MSTR_mantle_1}
    translation {y=6}
   }
   n12 = RigidBody!(ComponentPose=@n11, Alpha:float=@n10, OverrideWorldGravity="0,0,0", ExternalForce) -> (Pose) {
    bOverrideWorldGravity=true
    overridePhysicsAsset=&/M_MED_BattleHound_Fierce_Cape_Physics
    simulationSpace=BaseBoneSpace
    alphaBoolBlend {blendOption=Linear}
    baseBoneRef {boneName=spine_05}
    componentAppliedLinearAccClamp=<0 0 0>
    simSpaceSettings {
     maxAngularAcceleration=50
     maxAngularVelocity=3
     maxLinearAcceleration=600
     maxLinearVelocity=700
     velocityScaleZ=2
     worldAlpha=1
     externalAngularVelocity {z=0.1}
     externalLinearDragV=<0.2 0.2 0.2>
    }
   }
   Save cached pose 'Main'(Pose=Component To Local(ComponentPose=@n12)) {cacheName=Main}
  ]}
 EventGraph {nodes [
   Event BlueprintUpdateAnimation()
   > n1 = Gravity Override State Function DEPRECATED(
    CharacterPartAnimInstance=Self-Reference()
    GameTimeInSeconds=GetGameTimeInSeconds()
    GravityOverrideParams=Get Gravity Override Params()
    RBAN Params=Get RBAN Params()
   )
   > Set Gravity Override(Gravity Override=@n1)
   > n2 = Branch(Condition=Get HasBackpack() -> (HasBackpack:bool))
   > n3 = Branch(Condition=Get IsBattleHoundFireCape())
   n4 = Get BackPack_skelMesh() -> (BackPack_skelMesh:Skeletal Mesh Component Object Reference)
   n5 = Get Mesh(self=Get As Fort Player Pawn Athena() -> (As Fort Player Pawn Athena:Fort Player Pawn Athena Object Reference))
   n6 = Vector Length(A=vector - vector(
     A=GetSocketLocation!(self:Scene Component Object Reference=@n4, InSocketName:Name=dyn_cape04) -> (ReturnValue:Vector)
     B=GetSocketLocation(self=@n5, InSocketName=spine_03)
    ))
   > Set NoCapeDynamics_Alpha(NoCapeDynamics_Alpha:double=MapRangeClamped(Value=@n6, InRangeA=25, InRangeB=50, OutRangeA=1, OutRangeB=0)) -> (Output_Get:double)
   > Set CapeDynamics_Alpha(CapeDynamics_Alpha:double=MapRangeClamped(Value=@n6, InRangeA=25, InRangeB=50)) -> (Output_Get:double)
   > n7 = Set RotateCapeUp(RotateCapeUp:bool=Get bIsSkyDiving()) -> (Output_Get:bool)
   n8 = Branch(execute=[@n7.then, @n11.then], Condition=Get bIsOnGround() -> (bIsOnGround:bool))
   n9 = Get PawnVelocity()
   > Set isMoving(isMoving:double=MapRangeClamped(Value=Vector Length(A=@n9), InRangeA=0, InRangeB=100)) -> (Output_Get:double)
   Set isMoving(execute=@n8.else, isMoving=0)
   Set NoCapeDynamics_Alpha(execute=@n3.else, NoCapeDynamics_Alpha=1)
   > n10 = Set CapeDynamics_Alpha(CapeDynamics_Alpha=0)
   n11 = Set RotateCapeUp(execute=[@n10.then, @n12.then], RotateCapeUp=false)
   Set NoCapeDynamics_Alpha(execute=@n2.else, NoCapeDynamics_Alpha=0)
   > n12 = Set CapeDynamics_Alpha(CapeDynamics_Alpha=1)
   n13 = Event BlueprintInitializeAnimation()
   Set HasBackpack(execute=[@n13.then, @n17.then], HasBackpack:bool=CheckForBackbling()) -> (Output_Get:bool)
   > n14 = Set IsBattleHoundFireCape(IsBattleHoundFireCape:bool=BackpackItemDefinitionHasAnyMetaTag!(Tags:Array of Gameplay Tag Structures=Make Array("[0]"=(TagName="Cosmetics.MetaTags.BattleHoundFire")), bRequiresEquipped:bool=false) -> (ReturnValue:bool)) -> (Output_Get:bool)
   > Branch(Condition=@n14)
   > n15 = Cast To FortPlayerPawnAthena(Object=TryGetPawnOwner())
   > n16 = Set As Fort Player Pawn Athena(As Fort Player Pawn Athena:Fort Player Pawn Athena Object Reference=@n15) -> (Output_Get:Fort Player Pawn Athena Object Reference)
   > Set BackPack_skelMesh(BackPack_skelMesh:Skeletal Mesh Component Object Reference=GetSkeletalMeshForPartType(self=@n16, PartType=Backpack)) -> (Output_Get:Skeletal Mesh Component Object Reference)
   n17 = Event OnCharacterCustomizationLoaded()
  ]}
]
variables [
 As Fort Player Pawn Athena {type=FortPlayerPawnAthena}
 BackPack_skelMesh {type=SkeletalMeshComponent}
 CapeDynamics_Alpha {type=number}
 CapeDynamicsAlpha {type=number}
 Gravity Override {type=Vector}
 Gravity Override Params {type=GravityOverrideParamsStruct}
 HasBackpack {type=boolean}
 Is_LOD_0 {type=boolean}
 IsBattleHoundFireCape {type=boolean}
 isMoving {type=number}
 MoveDefaultCape {type=boolean}
 NoCapeDynamics {type=boolean}
 NoCapeDynamics_Alpha {type=number}
 RBAN Params {
  type=RBANParamsStruct
  default {
   dBNO {simSpaceSettings {
     maxAngularAcceleration=1000
     maxAngularVelocity=3
     maxLinearAcceleration=300
     maxLinearVelocity=700
     velocityScaleZ=0
     externalLinearDragV=<0 0.5 0>
    }}
   dBNO_Carried {simSpaceSettings {
     maxAngularAcceleration=1000
     maxAngularVelocity=3
     maxLinearAcceleration=300
     maxLinearVelocity=700
     velocityScaleZ=0
     externalLinearDragV=<0 0.5 0>
    }}
   emoteOrMelee {simSpaceSettings {
     maxAngularAcceleration=1000
     maxAngularVelocity=3
     maxLinearAcceleration=300
     maxLinearVelocity=700
     velocityScaleZ=0
     externalLinearDragV=<0 0.5 0>
    }}
   front_End {simSpaceSettings {
     maxAngularAcceleration=1000
     maxAngularVelocity=3
     maxLinearAcceleration=300
     maxLinearVelocity=700
     velocityScaleZ=0
     externalLinearDragV=<0 0.5 0>
    }}
   onGround_Crouch_Moving {simSpaceSettings {
     maxAngularAcceleration=1000
     maxAngularVelocity=3
     maxLinearAcceleration=300
     maxLinearVelocity=700
     velocityScaleZ=0
     externalLinearDragV=<0 0.5 0>
    }}
   onGround_Falling {simSpaceSettings {
     maxAngularAcceleration=1000
     maxAngularVelocity=3
     maxLinearAcceleration=300
     maxLinearVelocity=700
     velocityScaleZ=0
     externalLinearDragV=<0 0.5 0>
    }}
   onGround_Moving {simSpaceSettings {
     maxAngularAcceleration=1000
     maxAngularVelocity=3
     maxLinearAcceleration=300
     maxLinearVelocity=700
     velocityScaleZ=0
     externalLinearDragV=<0 0.5 0>
    }}
   onGround_Standing {simSpaceSettings {
     maxAngularAcceleration=1000
     maxAngularVelocity=3
     maxLinearAcceleration=300
     maxLinearVelocity=700
     velocityScaleZ=0
     externalLinearDragV=<0 0.5 0>
    }}
   rail_Grinding {simSpaceSettings {
     maxAngularAcceleration=1000
     maxAngularVelocity=3
     maxLinearAcceleration=300
     maxLinearVelocity=700
     velocityScaleZ=0
     worldAlpha=0.1
     externalLinearDragV=<0 0.5 0>
    }}
   riding_Moving {simSpaceSettings {
     maxAngularAcceleration=1000
     maxAngularVelocity=3
     maxLinearAcceleration=300
     maxLinearVelocity=700
     velocityScaleZ=0
     externalLinearDragV=<0 0.5 0>
    }}
   riding_Moving_Targeting {simSpaceSettings {
     maxAngularAcceleration=1000
     maxAngularVelocity=3
     maxLinearAcceleration=300
     maxLinearVelocity=700
     velocityScaleZ=0
     externalLinearDragV=<0 0.5 0>
    }}
   riding_Standing {simSpaceSettings {
     maxAngularAcceleration=1000
     maxAngularVelocity=3
     maxLinearAcceleration=300
     maxLinearVelocity=700
     velocityScaleZ=0
     externalLinearDragV=<0 0.5 0>
    }}
   skydive_DivingDown {simSpaceSettings {
     maxAngularAcceleration=1000
     maxAngularVelocity=3
     maxLinearAcceleration=300
     maxLinearVelocity=700
     velocityScaleZ=0
     externalLinearDragV=<0 0.5 0>
    }}
   skydive_FloatingDown {
    componentLinearAccScale=<0 0.5 0>
    componentLinearVelScale=<0 0.5 0>
    simSpaceSettings {
     maxAngularAcceleration=1000
     maxAngularVelocity=3
     maxLinearAcceleration=300
     maxLinearVelocity=700
     velocityScaleZ=0
     externalLinearDragV=<0 0.5 0>
    }
   }
   skydive_GoingUp {simSpaceSettings {
     maxAngularAcceleration=1000
     maxAngularVelocity=3
     maxLinearAcceleration=300
     maxLinearVelocity=700
     velocityScaleZ=0
     externalLinearDragV=<0 0.5 0>
    }}
   skydive_Parachuting {
    componentLinearAccScale=<0 0.1 0.3>
    componentLinearVelScale=<0 0.1 0.3>
    simSpaceSettings {
     maxAngularAcceleration=1000
     maxAngularVelocity=3
     maxLinearAcceleration=300
     maxLinearVelocity=700
     velocityScaleZ=0
     externalLinearDragV=<0 0.5 0>
    }
   }
   skydive_Surfing {simSpaceSettings {
     maxAngularAcceleration=1000
     maxAngularVelocity=3
     maxLinearAcceleration=300
     maxLinearVelocity=700
     velocityScaleZ=0
     externalLinearDragV=<0 0.5 0>
    }}
   swimming {simSpaceSettings {
     maxAngularAcceleration=1000
     maxAngularVelocity=3
     maxLinearAcceleration=300
     maxLinearVelocity=700
     velocityScaleZ=0
     externalLinearDragV=<0 0.5 0>
    }}
   tactical_Sprint {simSpaceSettings {
     maxAngularAcceleration=1000
     maxAngularVelocity=3
     maxLinearAcceleration=300
     maxLinearVelocity=700
     velocityScaleZ=0
     externalLinearDragV=<0 0.5 0>
    }}
   ziplining {simSpaceSettings {
     maxAngularAcceleration=1000
     maxAngularVelocity=3
     maxLinearAcceleration=300
     maxLinearVelocity=700
     velocityScaleZ=0
     externalLinearDragV=<0 0.5 0>
    }}
  }
 }
 RotateCapeUp {type=boolean}
]
