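# Collision-only physics asset for the Fierce BattleHound male medium cosmetic body, named as the colliders for its cape.
# Capsules cover the pelvis, spine, neck and clavicles, with extra overlapping shapes on spine_02 and spine_05 and a sphere for the head.
# Smaller capsules on bandage and skirt bones add accessory volume.
# Arm and leg bodies hold no shapes, and no constraints link bodies, so it defines no ragdoll joints.
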PhysicsAsset /BRCosmetics/Characters/Player/Male/Medium/Bodies/M_MED_BattleHound_Fierce/Meshes/M_MED_BattleHound_Fierce_CapeColliders
bodies {
 bandage_l_1 [{
   shapeType=Capsule
   length=5.125
   radius=2.963
   shapeName=bandage_l_1_capsule
   center=<3.6902 1.1546 0.16>
   rotation=<rot -1.1363 -69.2603 -87.8735>
  }]
 bandage_r_1 [{
   shapeType=Capsule
   length=9.875
   radius=2.197
   shapeName=bandage_r_1_capsule
   center=<6.2778 -0.6188 0.122>
   rotation=<rot -8.2365 75.598 -92.9771>
  }]
 calf_l []
 calf_r []
 clavicle_l [{shapeType=Capsule, length=10.0192, radius=13.1371, center=<11.8905 1.4102 0.1998>, rotation=<rot 69.3297 1.4476 176.9462>}]
 clavicle_r [{shapeType=Capsule, length=10.0192, radius=13.1371, center=<-11.3746 -1.3971 1.1679>, rotation=<rot 69.3297 1.4476 -3.0538>}]
 foot_l []
 foot_r []
 head [{shapeType=Sphere, radius=17.3125, center=<11.028 -0.174 0>}]
 lowerarm_l []
 lowerarm_r []
 neck_02 [{shapeType=Capsule, length=13.4375, radius=7.3175, center=<-0.2642 -0.3408 0>, rotation=<rot 90 -171.1193 -175.3978>}]
 pelvis [{shapeType=Capsule, length=6.0206, radius=17.1638, center=<-2.4887 0.1881 0>}]
 skirt_C_l_2 [{shapeType=Capsule, length=3.3961, radius=6.8429, shapeName=skirt_C_l_2_capsule, center=<3.9798 2.9966 0>, rotation=<rot 90 180 174.3748>}]
 skirt_C_r_2 [{shapeType=Capsule, length=4.4226, radius=6.9792, shapeName=skirt_C_r_2_capsule, center=<3.942 3.3808 0>, rotation=<rot 90 180 174.3749>}]
 skirt_mid_2 [{shapeType=Capsule, length=7.1836, radius=9.948, shapeName=skirt_mid_2_capsule, center=<4.8011 7.3902 0>, rotation=<rot -90 0.0001 0.0001>}]
 spine_01 [{shapeType=Capsule, length=5.8188, radius=17.9336, center=<-1.7965 0.1884 0>}]
 spine_02 [
  {shapeType=Capsule, length=7.3264, radius=17.6322, center=<-1.711 -0.1851 0.0001>}
  {shapeType=Capsule, length=29.3956, radius=11.5807, center=<2.5024 0.4992 3.9825>, rotation=<rot 90 0.0049 0.0049>}
  {shapeType=Capsule, length=29.875, radius=11.1013, center=<2.4704 0.4991 -3.9825>, rotation=<rot 90 -0.0396 -0.0396>}
 ]
 spine_03 [{shapeType=Capsule, length=9.3357, radius=18.0781, center=<0.1445 0.2396 0.0001>}]
 spine_04 [{shapeType=Capsule, length=15.5654, radius=17.8825, center=<2.766 0.7765 0.0002>}]
 spine_05 [
  {shapeType=Capsule, length=14.8976, radius=16.945, center=<1.3732 1.8109 0>}
  {shapeType=Capsule, length=13.9518, radius=13.9425, center=<-0.2916 0.4452 0>}
 ]
 thigh_l []
 thigh_r []
 upperarm_l []
 upperarm_r []
}
constraints []
